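# Niagara system for the TracePaper emote that emits grey smoke wisps for extinguished flames.
# A burst sized by User.FlameNumber is partitioned into per-flame particle chains, with sprite and ribbon renderers.
# Normalized place in chain drives lifetime, width, spread, rise and curl noise, so wisps widen and distort along their length while fading.
# Source XY movement is fully inherited for about a second, then only toward the chain's start.
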NiagaraSystem /BRCosmetics/Animation/Game/MainPlayer/Emotes/TracePaper/FX/NS_Emote_TracePaper_ExtinguishSmoke
dependencies {
 usedDataInterfaces [&NiagaraDataInterfaceCurve, &NiagaraDataInterfaceColorCurve]
 usedDynamicInputs []
 usedModules [&SystemState]
}
emitters {NE_Smoke {
  emitterSpawnScript [/SetVariables(Emitter.ParticlesPerFlame=20) {isSetParameters=true}]
  emitterUpdateScript [
   &EmitterState(Life Cycle Mode=::NewEnumerator1, Loop Behavior=::NewEnumerator1, Loop Duration=3, UseLoopDelay=false)
   &SpawnBurst_Instantaneous(
    Spawn Count=&Multiply_Int(A=$Emitter.ParticlesPerFlame, B=$User.FlameNumber)
    Use Spawn Probability=false
    UseLoopCountLimit=false
   )
  ]
  particleSpawnScript [
   &PartitionParticles(
    Number Of Partitions=$User.FlameNumber
    Particles Per Partition:NiagaraInt32=$Emitter.ParticlesPerFlame
    Partition Mode:ENiagara_PartitionMode=::NewEnumerator1
   )
   /SetVariables(Particles.NormalizedPlaceInChain:NiagaraFloat=$Particles.PartitionParticles.Normalized Index In Partition) {isSetParameters=true}
   &InitializeParticle(
    Color=<rgba 1 1 1 0.5>
    Lifetime=&Lerp_Float(A=2, B=3.5, Alpha=$Particles.NormalizedPlaceInChain)
    Position Offset=&Lerp_Vector(A=<0 0 0>, B=<0 0 0>, Alpha=$Particles.NormalizedPlaceInChain)
    Ribbon Width=20
    Ribbon Width Mode=::NewEnumerator1
    UsePositionOffset=true
    Write Material Random=true
   )
   &ShapeLocation(
    Cylinder Distribution=::NewEnumerator2
    Cylinder Height=1
    Cylinder Radius=&Lerp_Float(A=0, B=5, Alpha=$Particles.NormalizedPlaceInChain)
    MaskSpawnGroup=false
    Shape Primitive=::NewEnumerator1
    U Position=&Multiply_Float(A=2, B=$Particles.NormalizedPlaceInChain)
    UseOwnerScale=true
    V Position=$Particles.NormalizedPlaceInChain
   )
   /SetVariables(Particles.RibbonWidth=&FloatFromCurve(
     FloatCurve=DataInterface() {propertyValues {Curve {keys [
         {arriveTangent=1.692, interpMode=RCIM_Cubic, leaveTangent=1.692, tangentMode=RCTM_User, time=0, value=0}
         {arriveTangent=3.2787, tangentMode=RCTM_User, time=1, value=3}
        ]}}}
     CurveIndex=$Particles.NormalizedPlaceInChain
     Scale Curve=2
    )) {isSetParameters=true}
   &AddVelocity(Velocity=&Lerp_Vector(A=<0 0 0>, B=<0 0 20>, Alpha=$Particles.NormalizedPlaceInChain))
  ]
  particleUpdateScript [
   &ParticleState()
   /SetVariables(Particles.RibbonU0Override=$Particles.NormalizedPlaceInChain) {isSetParameters=true}
   /SetVariables(Particles.RibbonU1Override:NiagaraFloat=&Multiply_Float(A=$Particles.NormalizedPlaceInChain, B=4)) {isSetParameters=true}
   &CurlNoiseForce(
    Noise Frequency=250
    Noise Strength=&Multiply_Float(
     A=&Lerp_Float(
      A=0
      B=100
      Alpha=&FloatFromCurve(
       FloatCurve=DataInterface() {propertyValues {Curve {keys [
           {arriveTangent=0.3198, interpMode=RCIM_Cubic, leaveTangent=0.3198, tangentMode=RCTM_User, time=0, value=0}
           {arriveTangent=3.6292, interpMode=RCIM_Cubic, leaveTangent=3.6292, tangentMode=RCTM_User, time=0.7, value=1}
          ]}}}
       CurveIndex=$Particles.NormalizedPlaceInChain
      )
     )
     B=&FloatFromCurve(FloatCurve=DataInterface() {propertyValues {Curve {keys [
          {arriveTangent=3.3096, interpMode=RCIM_Cubic, leaveTangent=3.3096, tangentMode=RCTM_User, time=0, value=0}
          {interpMode=RCIM_Cubic, tangentMode=RCTM_User, time=1, value=1}
         ]}}})
    )
    Pan Noise=false
   )
   &AccelerationForce(Acceleration=&Lerp_Vector(A=<0 0 5>, B=<0 0 45>, Alpha=$Particles.NormalizedPlaceInChain))
   &InheritSourceMovement(
    Applied Position Scale=&Multiply_VectorByFloat(
     Vector=<1 1 0>
     Float=&Lerp_Float(
      A=&FloatFromCurve(
       FloatCurve=DataInterface() {propertyValues {Curve {keys [
           {arriveTangent=-3.4409, interpMode=RCIM_Cubic, leaveTangent=-3.4409, tangentMode=RCTM_User, time=0, value=1}
           {interpMode=RCIM_Cubic, tangentMode=RCTM_User, time=1, value=0}
          ]}}}
       CurveIndex=$Particles.NormalizedPlaceInChain
      )
      B=1
      Alpha=&FloatFromCurve(FloatCurve=DataInterface() {propertyValues {Curve {keys [<at 1 1>,<at 1.1 0>]}}}, CurveIndex=$Emitter.Age)
     )
    )
    ApplyForce=false
    ApplyPosition=true
    ApplyPositionFalloff=false
   )
   &SolveForcesAndVelocity(Clamp Velocity=false, Limit Acceleration=false)
   &DynamicMaterialParameters(
    Index 1 Param 3=&Max_Float(
     A:NiagaraFloat=&FloatFromCurve(FloatCurve=DataInterface() {propertyValues {Curve {keys [<at 0.5 0.7>,<at 1 1>]}}}, CurveIndex=$Particles.NormalizedPlaceInChain)
     B:NiagaraFloat=&FloatFromCurve(FloatCurve=DataInterface() {propertyValues {Curve {keys [<at 0.5 0>,<at 0.7 1>]}}})
    )
    Module.Index0Param0Write=true
    Module.Index0Param1Write=true
    Module.Index0Param2Write=true
    Module.Index0Param3Write=true
    Module.Index1Param0Write=true
    Module.Index1Param1Write=true
    Module.Index1Param2Write=true
    Module.Index1Param3Write=true
    Write Parameter Index 1=true
   )
   &ScaleRibbonWidth(Ribbon Width Scale=&FloatFromCurve(FloatCurve=DataInterface() {propertyValues {Curve {keys [
         {arriveTangent=-0.288, interpMode=RCIM_Cubic, leaveTangent=-0.288, tangentMode=RCTM_User, time=0, value=0}
         {arriveTangent=-0.0827, interpMode=RCIM_Cubic, leaveTangent=-0.0827, tangentMode=RCTM_User, time=1, value=3}
        ]}}}))
   &Color(Color=&ColorFromCurve(ColorCurve=DataInterface() {propertyValues {
       AlphaCurve {keys [<at 0.2783 1 RCIM_Cubic>,<at 1 0 RCIM_Cubic>]}
       BlueCurve {keys [<at 0 0.2031 RCIM_Cubic>]}
       GreenCurve {keys [<at 0 0.2031 RCIM_Cubic>]}
       RedCurve {keys [<at 0 0.2031 RCIM_Cubic>]}
      }}))
  ]
  rendererClasses [&NiagaraSpriteRendererProperties, &NiagaraRibbonRendererProperties]
  renderers [
   {rendererClass=&NiagaraSpriteRendererProperties}
   {rendererClass=&NiagaraRibbonRendererProperties, rendererIndex=1}
  ]
 }}
summary {
 systemName=NS_Emote_TracePaper_ExtinguishSmoke
 emitters [{emitterName=NE_Smoke, rendererClasses [&NiagaraSpriteRendererProperties, &NiagaraRibbonRendererProperties]}]
 userVariables [User.FlameNumber {type=NiagaraInt32, defaultValue=1}]
}
