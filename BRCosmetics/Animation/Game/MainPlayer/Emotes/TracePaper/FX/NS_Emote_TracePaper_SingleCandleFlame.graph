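# Niagara candle flame effect for the TracePaper emote, with User.FlameNumber setting the flame count and User.FadeFlame scaling it in and out.
# NE_Flame builds each flame as a short partitioned ribbon chain, tapered by curve, whose tip bends against owner velocity and lags through time-smoothed goal positions.
# NE_Glow places a sprite per flame along that ribbon via a particle attribute reader.
# NE_LightRenderer adds a flickering, time-of-day tinted light.
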NiagaraSystem /BRCosmetics/Animation/Game/MainPlayer/Emotes/TracePaper/FX/NS_Emote_TracePaper_SingleCandleFlame
dependencies {
 usedDataInterfaces [&NiagaraDataInterfaceCurve, &NiagaraDataInterfaceParticleRead]
 usedDynamicInputs []
 usedModules [&SystemState]
}
emitters {
 NE_Flame {
  emitterSpawnScript [/SetVariables(Emitter.ParticlesPerFlame=7) {isSetParameters=true}]
  emitterUpdateScript [
   &EmitterState(
    Inactive Response=::NewEnumerator1
    Life Cycle Mode=::NewEnumerator1
    Loop Behavior=::NewEnumerator1
    Loop Duration Mode=::NewEnumerator1
    UseLoopDelay=false
   )
   &SpawnBurst_Instantaneous(
    Spawn Count=&Multiply_Int(A=$Emitter.ParticlesPerFlame, B=$User.FlameNumber)
    Use Spawn Probability=false
    UseLoopCountLimit=false
   )
  ]
  particleSpawnScript [
   &InitializeParticle(Lifetime=5, Ribbon Width=20, Ribbon Width Mode=::NewEnumerator1, UsePositionOffset=false, Write Material Random=true)
   &PartitionParticles(
    Number Of Partitions=$User.FlameNumber
    Particles Per Partition:NiagaraInt32=$Emitter.ParticlesPerFlame
    Partition Mode:ENiagara_PartitionMode=::NewEnumerator1
   )
   /SetVariables(Particles.NormalizedPlaceInChain:NiagaraFloat=$Particles.PartitionParticles.Normalized Index In Partition) {isSetParameters=true}
   /SetVariables(Particles.RibbonWidth=&FloatFromCurve(
     FloatCurve=DataInterface() {propertyValues {Curve {keys [
         {arriveTangent=8.9306, interpMode=RCIM_Cubic, leaveTangent=8.9306, tangentMode=RCTM_User, time=0, value=0.3}
         {interpMode=RCIM_Cubic, tangentMode=RCTM_User, time=0.2, value=1}
         {arriveTangent=-1.0764, interpMode=RCIM_Cubic, leaveTangent=-1.0764, tangentMode=RCTM_User, time=1, value=0}
        ]}}}
     CurveIndex=$Particles.NormalizedPlaceInChain
     Scale Curve=4
    )) {isSetParameters=true}
   /SetVariables(Particles.GlowWidth:NiagaraFloat=&Multiply_Float(A=$Particles.RibbonWidth, B=2)) {isSetParameters=true}
  ]
  particleUpdateScript [
   &ParticleState(Kill Particles When Lifetime Has Elapsed=false)
   /SetVariables(Particles.RibbonU0Override=$Particles.NormalizedPlaceInChain) {isSetParameters=true}
   /SetVariables(Particles.VelocityBender:Vector3f=&Multiply_VectorByFloat(
     Vector=/Niagara/DynamicInputs/Math/Clamp_Vector(
      Vector:Vector3f=&Multiply_VectorByFloat(
       Vector=&TransformVector(Vector=$Engine.Owner.Velocity, Source Space=::World, Destination Space=::Local)
       Float=-0.05
      )
      Min:Vector3f=&VectorFromFloat(Value=-2.5)
      Max:Vector3f=&VectorFromFloat(Value=2.5)
     )
     Float=&FloatFromCurve(
      FloatCurve=DataInterface() {propertyValues {Curve {keys [
          {arriveTangent=0.1033, interpMode=RCIM_Cubic, leaveTangent=0.1033, tangentMode=RCTM_User, time=0.2, value=0}
          {interpMode=RCIM_Cubic, tangentMode=RCTM_User, time=1, value=1}
         ]}}}
      CurveIndex=$Particles.NormalizedPlaceInChain
     )
    )) {isSetParameters=true}
   /SetVariables(Particles.GoalPosition:NiagaraPosition=&AddVectorToPosition(
     Position=<0 0 0>
     Vector=&Add_Vector(
      A=&Lerp_Vector(
       A=<0 0 0>
       B=<0 0 10>
       Alpha=&FloatFromCurve(FloatCurve=DataInterface() {propertyValues {Curve {keys [<at 0 0>,<at 1 1>]}}}, CurveIndex=$Particles.NormalizedPlaceInChain)
      )
      B=$Particles.VelocityBender
     )
    )) {isSetParameters=true}
   /SetVariables(Particles.GlowGoalPos:NiagaraPosition=&AddVectorToPosition(
     Position=<0 0 0>
     Vector=&AddVectorToPosition(
      Position=&Lerp_Vector(
       A=<0 0 0>
       B=<0 0 17>
       Alpha=&FloatFromCurve(FloatCurve=DataInterface() {propertyValues {Curve {keys [<at 0 0>,<at 1 1>]}}}, CurveIndex=$Particles.NormalizedPlaceInChain)
      )
      Vector=$Particles.VelocityBender
     )
    )) {isSetParameters=true}
   &InterpolateOverTime(Rate Of Change=10, Target Vector Value:NiagaraPosition=$Particles.GoalPosition, Variable Type=::NewEnumerator2)
   &InterpolateOverTime(Rate Of Change=10, Target Vector Value:NiagaraPosition=$Particles.GlowGoalPos, Variable Type=::NewEnumerator2) {module=InterpolateOverTime001}
   &InheritSourceMovement(Applied Position Scale=<-1 -1 -1>, ApplyForce=false, ApplyPosition=true, ApplyPositionFalloff=false)
   /SetVariables(Particles.Position=&Lerp_Vector(
     A=<0 0 0>
     B=&Lerp_Vector(
      A=Linked() {linkedVariable {name=Particles.GoalPosition, type=NiagaraPosition}}
      B=$Particles.InterpolateOverTime.InterpolateOverTime_V3_.Moving Average
      Alpha=&FloatFromCurve(
       FloatCurve=DataInterface() {propertyValues {Curve {keys [
           {arriveTangent=0.7977, interpMode=RCIM_Cubic, leaveTangent=0.7977, tangentMode=RCTM_User, time=0.1, value=0}
           {arriveTangent=3.1196, interpMode=RCIM_Cubic, leaveTangent=3.1196, tangentMode=RCTM_User, time=1, value=1}
          ]}}}
       CurveIndex=$Particles.NormalizedPlaceInChain
      )
     )
     Alpha=$User.FadeFlame
    )) {isSetParameters=true}
   /SetVariables(Particles.GlowPos:NiagaraPosition=&Lerp_Vector(
     A=<0 0 0>
     B=&Lerp_Vector(
      A=Linked() {linkedVariable {name=Particles.GoalPosition, type=NiagaraPosition}}
      B=$Particles.InterpolateOverTime001.InterpolateOverTime_V3_.Moving Average
      Alpha=&FloatFromCurve(
       FloatCurve=DataInterface() {propertyValues {Curve {keys [
           {arriveTangent=0.7977, interpMode=RCIM_Cubic, leaveTangent=0.7977, tangentMode=RCTM_User, time=0.1, value=0}
           {arriveTangent=3.1196, interpMode=RCIM_Cubic, leaveTangent=3.1196, tangentMode=RCTM_User, time=1, value=1}
          ]}}}
       CurveIndex=$Particles.NormalizedPlaceInChain
      )
     )
     Alpha=$User.FadeFlame
    )) {isSetParameters=true}
   &SolveForcesAndVelocity(Clamp Velocity=false, Limit Acceleration=false)
   &DynamicMaterialParameters(
    Index 1 Param 3=&FloatFromCurve(FloatCurve=DataInterface() {propertyValues {Curve {keys [<at 0.5 0>,<at 1 1>]}}}, CurveIndex=$Particles.NormalizedPlaceInChain)
    Module.Index0Param0Write=true
    Module.Index0Param1Write=true
    Module.Index0Param2Write=true
    Module.Index0Param3Write=true
    Module.Index1Param0Write=true
    Module.Index1Param1Write=true
    Module.Index1Param2Write=true
    Module.Index1Param3Write=true
    Write Parameter Index 1=true
   )
   &ScaleColor(Scale Alpha=$User.FadeFlame, ScaleA=true, ScaleRGB=true)
   &FN_Light_Attributes(
    Color=&FNTimeOfDayColor(Night Color=<rgba 1 0.1249 0 1>, Morning Color=<rgba 1 0.1961 0 1>, Day Color=<rgba 1 0.2621 0 1>, Evening Color=<rgba 1 0.1329 0 1>)
    Create Color Attribute=true
    Create Exponent Attribute=true
    Create Radius Attribute=true
    Enable Particle Light Enable=true
    Exponent:NiagaraFloat=&Add_Float(A=10, B=&Sine(Normalized Angle=$Emitter.Age, Period=0.7, Scale=1.5, Bias=0))
    Light Enabled=true
    Radius=&Add_Float(A=115, B=&Sine(Normalized Angle=$Emitter.Age, Period=0.7, Scale=4, Bias=0))
    Scale Color=50
   )
  ]
  rendererClasses [&NiagaraSpriteRendererProperties, &NiagaraRibbonRendererProperties]
  renderers [
   {rendererClass=&NiagaraSpriteRendererProperties}
   {rendererClass=&NiagaraRibbonRendererProperties, rendererIndex=1}
   {rendererClass=&NiagaraRibbonRendererProperties, rendererIndex=2}
  ]
 }
 NE_Glow {
  emitterUpdateScript [
   &EmitterState(
    Inactive Response=::NewEnumerator1
    Life Cycle Mode=::NewEnumerator1
    Loop Behavior=::NewEnumerator1
    Loop Duration Mode=::NewEnumerator1
    UseLoopDelay=false
   )
   &SpawnBurst_Instantaneous(Spawn Count=$User.FlameNumber, Use Spawn Probability=false, UseLoopCountLimit=false)
  ]
  particleSpawnScript [
   &InitializeParticle(
    Color=<rgba 5 2.2737 0 0.3>
    Position Offset=<0 0 4>
    Sprite Rotation Angle=0.25
    Sprite Rotation Mode=::NewEnumerator1
    Sprite Size=<11 11>
    Sprite Size Mode=::NewEnumerator4
    UsePositionOffset=true
    Write Material Random=true
   )
   &PartitionParticles(Number Of Partitions=$User.FlameNumber)
  ]
  particleUpdateScript [
   &ParticleState(Kill Particles When Lifetime Has Elapsed=false)
   &self:StickMeshToRibbon_01(
    Index=3
    Particle Attribute Reader=DataInterface() {propertyValues {EmitterBinding {bindingMode=Other, emitterName=NE_Flame}}}
    ParticlePerRibbon=7
    Partition:NiagaraInt32=$Particles.PartitionParticles.Partition
    RibbonSpace=1
   ) {module=StickMeshToRibbon_01}
   &ScaleSpriteSize(
    Uniform Curve Scale=&Multiply_Float(A=$User.FadeFlame, B=1)
    Uniform Curve Sprite Scale=DataInterface() {propertyValues {Curve {keys [{tangentMode=RCTM_User, time=0, value=1}, {tangentMode=RCTM_User, time=1, value=1}]}}}
   )
   &DynamicMaterialParameters(
    Index 0 Param 1=0.3
    Module.Index0Param0Write=true
    Module.Index0Param1Write=true
    Module.Index0Param2Write=true
    Module.Index0Param3Write=true
   )
   &SpriteFacingAndAlignment(Sprite Alignment=$Particles.RibbonDirectionVector, WriteAlignment=true, WriteFacing=false)
  ]
  rendererClasses [&NiagaraSpriteRendererProperties]
  renderers [{rendererClass=&NiagaraSpriteRendererProperties}]
 }
 NE_LightRenderer {
  emitterUpdateScript [
   &EmitterState(
    Inactive Response=::NewEnumerator1
    Life Cycle Mode=::NewEnumerator1
    Loop Behavior=::NewEnumerator1
    Loop Duration Mode=::NewEnumerator1
    UseLoopDelay=false
   )
   &SpawnBurst_Instantaneous(Use Spawn Probability=false, UseLoopCountLimit=false)
  ]
  particleSpawnScript [&InitializeParticle(
    Color Mode=::NewEnumerator0
    Lifetime=5
    Position Offset=<0 0 4>
    Sprite Size Mode=::NewEnumerator3
    UsePositionOffset=true
    Write Material Random=true
   )]
  particleUpdateScript [
   &ParticleState(Kill Particles When Lifetime Has Elapsed=false)
   &FN_Light_Attributes(
    Color=&FNTimeOfDayColor(Night Color=<rgba 1 0.439 0 1>, Morning Color=<rgba 1 0.439 0 1>, Day Color=<rgba 1 0.4301 0 1>, Evening Color=<rgba 1 0.439 0 1>)
    Create Color Attribute=true
    Create Exponent Attribute=true
    Create Radius Attribute=true
    Enable Particle Light Enable=true
    Exponent:NiagaraFloat=&Add_Float(
     A=4
     B=&Waveform("Waveform [1]"=::NewEnumerator0, "[1] Frequency"=0.6, "[1] Amplitude Min/Max"=<0 1>, "[1] Scale"=1.5)
    )
    Light Enabled=true
    Radius=&Add_Float(
     A=&Multiply_Float(A=30, B=$User.FadeFlame)
     B=&Sine(Normalized Angle=$Emitter.Age, Period=0.7, Scale=3, Bias=0)
    )
    Scale Color=5
   )
  ]
  rendererClasses [&NiagaraSpriteRendererProperties, &NiagaraLightRendererProperties]
  renderers [
   {rendererClass=&NiagaraSpriteRendererProperties}
   {rendererClass=&NiagaraLightRendererProperties, rendererIndex=1}
  ]
 }
}
summary {
 systemName=NS_Emote_TracePaper_SingleCandleFlame
 emitters [
  {emitterName=NE_Flame, rendererClasses [&NiagaraSpriteRendererProperties, &NiagaraRibbonRendererProperties]}
  {
   emitterName=NE_LightRenderer
   rendererClasses [&NiagaraSpriteRendererProperties, &NiagaraLightRendererProperties]
  }
  {emitterName=NE_Glow, rendererClasses [&NiagaraSpriteRendererProperties]}
 ]
 userVariables [
  User.FadeFlame {type=NiagaraFloat, defaultValue=1, description=""}
  User.FlameNumber {type=NiagaraInt32, defaultValue=1}
 ]
}
